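; Ruckus Mini pickaxe weapon Blueprint whose glass globe shows as broken after an impact.
; BeginPlay creates dynamic materials for the globe and axe, and a weapon visibility change resets the counter, broken flag and glass.
; The impact handler checks broken state, frontend context and surface type around a replicated ServerCounter event that increments Counter.
; OnRep_Counter prints the count and, once it reaches one, sets Broken and the GlassBroken material parameter.
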(event EventBeginPlay
  ;[0.5688 1 0.1973 1 100] BEGINPLAY- REFS
  (bind _returnvalue (Rendering|Material|CreateDynamicMaterialInstance (Variables|Visual|GetWeaponMesh) 0 "/BRCosmetics/Weapons/FORT_Melee/Pickaxe_Ruckus_Mini/Materials/M_Pickaxe_Ruckus_Mini_Glass_Inst.M_Pickaxe_Ruckus_Mini_Glass_Inst" "None" :!class /Script/Engine.PrimitiveComponent))
  (Variables|Default|SetGlobeMat _returnvalue)
  (bind _returnvalue_2 (Rendering|Material|CreateDynamicMaterialInstance (Variables|Visual|GetWeaponMesh) 1 "/BRCosmetics/Weapons/FORT_Melee/Pickaxe_Ruckus_Mini/Materials/MI_Pickaxe_Ruckus_Mini.MI_Pickaxe_Ruckus_Mini" "None" :!class /Script/Engine.PrimitiveComponent))
  (Variables|Default|SetAxeMat _returnvalue_2))

(event EventOnWeaponVisibilityChanged
  (params (bVisible :type Boolean) (bSetForLocalControllerOnly :type Boolean))
  ;[1 1 1 1 100] ON DRAW - RESET COUNTER
  (|CallParentFunction bVisible bSetForLocalControllerOnly)
  (Rendering|Material|SetScalarParameterValue (Variables|Default|GetGlobeMat) "GlassBroken" 0.0 :!class /Script/Engine.MaterialInstanceDynamic)
  (Variables|Default|SetCounter 0)
  (Variables|Default|SetBroken false))

(event FortWeapon|EventOnPlayImpactFX
  (params (HitResult :type Hit%20Result%20Structure%20%28by%20ref%29) (ImpactPhysicalSurface :type EPhysicalSurface%20Enum :default "SurfaceType_Default") (SpawnedPSC :type FXSystem%20Component%20Object%20Reference))
  ;[1 0 0 1 100] BREAK ON IMPACT
  (if (not (Variables|Default|GetBroken))
    (if (not (Reactive|Frontend|IsCosmeticObjectDisplayedinFrontendContext self))
      (switch Utilities|FlowControl|Switch|SwitchonEPhysicalSurface ImpactPhysicalSurface
        (:SurfaceType12
          _)
        (:SurfaceType14
          _)
        (:SurfaceType15
          _)
        (:SurfaceType16
          _)))
    (CallFunction|ServerCounter)))

(event Custom|ServerCounter
  ;[1 0 0 1 100] BREAK ON IMPACT
  ;;[1 1 1] Replication
  (Variables|Default|SetCounter (+ (Variables|Default|GetCounter) 1)))

(fn OnRep_Counter
  (bind _counter (Variables|Default|GetCounter))
  (Networking|CanExecuteCosmeticEvents
    (:True
      (Development|PrintString (Utilities|String|ToString(Integer) _counter))
      (bind _output_get (Variables|Default|SetBroken (>= _counter 1)))
      (if _output_get
        (Rendering|Material|SetScalarParameterValue (Variables|Default|GetGlobeMat) "GlassBroken" 1.0 :!class /Script/Engine.MaterialInstanceDynamic)))))

(fn ConstructionScript
  (|CallParentFunction))
