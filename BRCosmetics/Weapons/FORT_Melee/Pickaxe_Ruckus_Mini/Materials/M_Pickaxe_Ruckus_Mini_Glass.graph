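# Translucent lit glass material for the Ruckus Mini pickaxe, with a stylized outline and a switchable broken state.
# Fresnel drives two-tone color, opacity and a high-contrast opaque outline whose exponent varies with camera distance.
# A sphere-masked sun/moon reflection adds an emissive highlight outside the outline.
# GlassBroken and vertex-color green select which vertices world position offset collapses to the object origin, before a shared TechArt override function.
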Material /BRCosmetics/Weapons/FORT_Melee/Pickaxe_Ruckus_Mini/Materials/M_Pickaxe_Ruckus_Mini_Glass
graph {
 properties {
  bEnableResponsiveAA=true
  bScreenSpaceReflections=true
  bUseMaterialAttributes=true
  bUsedWithMeshDeformer=true
  bUsedWithSkeletalMesh=true
  blendMode=BLEND_Translucent
  ditherOpacityMask=true
  pixelDepthOffsetMode=PDOM_Legacy
  translucencyDirectionalLightingIntensity=5
  translucencyLightingMode=TLM_VolumetricPerVertexNonDirectional
  usedShadingModels=Default Lit
 }
 nodes [
  n1 = ScalarParameter() {defaultValue=0.1, parameterName=Fresnel Exponent_Outline2}
  n2 = LinearInterpolate(
   A=VectorParameter() {parameterName=Base Color, defaultValue=<rgba 0.2655 0.3672 0.6042 0>}.RGB
   B=VectorParameter() {parameterName=Base Color2, defaultValue=<rgba 0.7161 0.7591 0.8594 0>}.RGB
   Alpha=MaterialFunctionCall(In=Fresnel(ExponentIn=@n1), Contrast=Constant() {r=50}) {materialFunction=&CheapContrast}.Result
  )
  n3 = VectorParameter() {parameterName=Outline Color}
  n4 = ScalarParameter() {defaultValue=100, group=NeonDepth, parameterName=DF_FarDistance}
  n5 = Subtract(A=Distance(A=WorldPosition().XYZ, B=CameraPositionWS()), B=@n4)
  n6 = Subtract(A=ScalarParameter() {defaultValue=500, group=NeonDepth, parameterName=DF_NearDistance}, B=@n4)
  n7 = LinearInterpolate(
   A=ScalarParameter() {defaultValue=50, group=NeonDepth, parameterName=DF_LineThickness_Near}
   B=ScalarParameter() {defaultValue=1, group=NeonDepth, parameterName=DF_LineThickness_Far}
   Alpha=Saturate(None=Divide(A=@n5, B=@n6))
  )
  n8 = ScalarParameter() {defaultValue=0.1, parameterName=Fresnel Exponent_Outline}
  n9 = MaterialFunctionCall(In=Fresnel(ExponentIn=Multiply(A=@n7, B=@n8)), Contrast=Constant() {r=50}) {materialFunction=&CheapContrast}
  n10 = CollectionParameter() {
   collection=&FortniteMaterialParameters
   parameterName=SunAndMoonModelDirectionalVector
  }
  n11 = ComponentMask(None=@n10) {b=true, g=true, r=true}
  n12 = ReflectionVectorWS()
  n13 = Power(Base=Clamp(None=DotProduct(A=@n11, B=@n12)), Exp=ScalarParameter() {defaultValue=70, parameterName=Spec Highlight Power}) {constExponent=20}
  n14 = Constant() {r=1}
  n15 = ScalarParameter() {defaultValue=1, parameterName=Spec Highlight Radius}
  n16 = ScalarParameter() {defaultValue=1, parameterName=Spec Highlight Hardness}
  n17 = Add(B=Multiply(A=SphereMask(A=@n13, B=@n14, Radius=@n15, Hardness=@n16), B=ScalarParameter() {defaultValue=3, parameterName=Spec Intense}))
  n18 = LinearInterpolate(
   A=ScalarParameter() {parameterName=Fresnel Inner Transparency_BaseGlass}
   B=ScalarParameter() {defaultValue=0.85, parameterName=Fresnel Outer Transparency_BaseGlass}
   Alpha=Fresnel(ExponentIn=ScalarParameter() {defaultValue=0.5, parameterName=Fresnel Exponent_BaseGlass})
  )
  n19 = Multiply(A=Subtract(A=WorldPosition().XYZ, B=ObjectPositionWS())) {constB=-1}
  n20 = Step(Y=VertexColor().G) {constX=0}
  n21 = Saturate(None=ScalarParameter() {parameterName=GlassBroken})
  n22 = LinearInterpolate(A=@n20, B=OneMinus(None=@n20), Alpha=OneMinus(None=Round(None=@n21)))
  n23 = SetMaterialAttributes!(
   MaterialAttributes
   Base Color=NamedRerouteDeclaration(None=LinearInterpolate(A=@n2, B=@n3.RGB, Alpha=@n9.Result)) {name=basecololr, nodeColor=<rgba 1 0.3281 0 1>}
   Emissive Color=NamedRerouteDeclaration(None=LinearInterpolate(A=@n17, Alpha=@n9.Result) {constB=0}) {name=emissive, nodeColor=<rgba 1 0.8203 0 1>}
   Opacity=NamedRerouteDeclaration(None=LinearInterpolate(A=@n18, Alpha=@n9.Result)) {name=Opacity, nodeColor=<rgba 0.7813 0 1 1>}
   World Position Offset=NamedRerouteDeclaration(None=Multiply(A=@n19, B=@n22)) {name=wpo, nodeColor=<rgba 0.5 1 0 1>}
  ) {attributeSetTypes [
    #1
    #2
    #3
    #4
   ]}
  MaterialFunctionCall(In=@n23) {materialFunction=&MF_TechArtOverrides_accessories}
 ]
}
