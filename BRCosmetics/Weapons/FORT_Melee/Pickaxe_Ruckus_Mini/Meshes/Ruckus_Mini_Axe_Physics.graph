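# Physics asset for the Ruckus Mini pickaxe, giving its jointed creature-like skeleton collision bodies and joints for physical motion.
# Two chains branch from the handle capsule: a spine through chest and neck to a large head sphere at neck_01, and a stiffer four-segment tail.
# Tight twist limits keep both chains bending rather than spinning.
# A short tube chain hangs off the chest, nearly rigid at its base with free swing and twist at dyn_tube_mid.
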PhysicsAsset /BRCosmetics/Weapons/FORT_Melee/Pickaxe_Ruckus_Mini/Meshes/Ruckus_Mini_Axe_Physics
bodies {
 Melee_TwoHanded_Handle [{shapeType=Capsule, length=47.75, radius=7.718, center=<-0.008 -1.1328 14.4302>, rotation=<rot -0.0367 -89.9999 0>}]
 chest [{
   shapeType=Capsule
   length=7.5419
   radius=3.9118
   shapeName=chest_capsule
   center=<0.286 -0.1142 -0.0659>
   rotation=<rot -1.0897 90.0619 -87.4449>
  }]
 dyn_tube_03 [{
   shapeType=Capsule
   length=6.1897
   radius=1.3413
   shapeName=dyn_tube_03_capsule
   center=<-0.4983 0.0528 0.0007>
   rotation=<rot -0.0084 -95.9986 -89.9855>
  }]
 dyn_tube_mid [{shapeType=Capsule, length=4.25, radius=2.005, shapeName=dyn_tube_mid_capsule, center=<1.5345 0 -0.2061>, rotation=<rot 0 -89.9999 -82.3385>}]
 neck [{shapeType=Capsule, length=1.3555, radius=3.9028, shapeName=neck_capsule, center=<5.7282 -0.09 0.153>, rotation=<rot 0.4632 -89.7387 -89.786>}]
 neck_01 [{shapeType=Sphere, radius=14.75, shapeName=neck_01_sphere, center=<18.4745 -0.1547 0>}]
 spine_02 [{shapeType=Capsule, length=4.25, radius=3, shapeName=spine_02_capsule, center=<1.2569 -0.1368 0.0946>, rotation=<rot 0.0014 -89.9997 91.7115>}]
 spine_03 [{
   shapeType=Capsule
   length=6.5847
   radius=3.0088
   shapeName=spine_03_capsule
   center=<1.6135 -0.2558 0.0097>
   rotation=<rot -0.0652 -89.3463 -89.8618>
  }]
 spine_04 [{shapeType=Capsule, length=8.0012, radius=3.4559, shapeName=spine_04_capsule, center=<2.2225 -0.2061 0>, rotation=<rot 0 90.1245 -89.9998>}]
 spine_05 [{
   shapeType=Capsule
   length=5.255
   radius=3.505
   shapeName=spine_05_capsule
   center=<1.3293 -0.0934 0.0425>
   rotation=<rot -78.8507 -84.6957 -94.3143>
  }]
 spine_07 [{shapeType=Capsule, length=3.5, radius=3.5, shapeName=spine_07_capsule, center=<2.1867 -0.2945 -0.0566>, rotation=<rot 0 -89.9999 89.9999>}]
 tail_02 [{shapeType=Capsule, length=7.372, radius=1.9729, shapeName=tail_02_capsule, center=<5.629 0.1376 0>, rotation=<rot 0 -89.7746 -89.9999>}]
 tail_03 [{shapeType=Capsule, length=6.729, radius=1.6349, shapeName=tail_03_capsule, center=<7.8138 0.1522 0>, rotation=<rot 0 -89.714 -90>}]
 tail_04 [{shapeType=Capsule, length=6.8482, radius=1.9291, shapeName=tail_04_capsule, center=<6.9306 0.1737 0>, rotation=<rot 0 -89.6816 -90>}]
 tail_05 [{shapeType=Capsule, length=9.0005, radius=3.2052, shapeName=tail_05_capsule, center=<7.9309 0.1353 0>, rotation=<rot 0 -90.3337 -90>}]
}
constraints [
 {bone1Name=spine_02, bone2Name=Melee_TwoHanded_Handle, swing1LimitDegrees=20, swing2LimitDegrees=20, twistLimitDegrees=1.5}
 {bone1Name=spine_03, bone2Name=spine_02, swing1LimitDegrees=25, swing2LimitDegrees=25, twistLimitDegrees=1}
 {bone1Name=spine_04, bone2Name=spine_03, swing1LimitDegrees=25, swing2LimitDegrees=25, twistLimitDegrees=1}
 {bone1Name=spine_05, bone2Name=spine_04, swing1LimitDegrees=25, swing2LimitDegrees=25, twistLimitDegrees=1}
 {bone1Name=chest, bone2Name=spine_05, swing1LimitDegrees=25, swing2LimitDegrees=25, twistLimitDegrees=2}
 {bone1Name=spine_07, bone2Name=chest, swing1LimitDegrees=25, swing2LimitDegrees=25, twistLimitDegrees=2}
 {bone1Name=neck, bone2Name=spine_07, swing1LimitDegrees=25, swing2LimitDegrees=25, twistLimitDegrees=2}
 {bone1Name=tail_02, bone2Name=Melee_TwoHanded_Handle, swing1LimitDegrees=10, swing2LimitDegrees=10, twistLimitDegrees=1}
 {bone1Name=tail_03, bone2Name=tail_02, swing1LimitDegrees=10, swing2LimitDegrees=10, twistLimitDegrees=1}
 {bone1Name=tail_04, bone2Name=tail_03, swing1LimitDegrees=10, swing2LimitDegrees=10, twistLimitDegrees=1}
 {bone1Name=tail_05, bone2Name=tail_04, swing1LimitDegrees=10, swing2LimitDegrees=10, twistLimitDegrees=1}
 {bone1Name=neck_01, bone2Name=neck, swing1LimitDegrees=25, swing2LimitDegrees=25, twistLimitDegrees=1}
 {bone1Name=dyn_tube_03, bone2Name=chest, swing1LimitDegrees=5, swing2LimitDegrees=3, twistLimitDegrees=2}
 {
  bone1Name=dyn_tube_mid
  bone2Name=dyn_tube_03
  swing1LimitDegrees=45
  swing1Motion=Free
  swing2LimitDegrees=45
  swing2Motion=Free
  twistLimitDegrees=45
  twistMotion=Free
 }
]
