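# Animation Blueprint for the Ruckus Mini Axe pickaxe cosmetic, animating the prop's idle and swing with secondary rigid body physics.
# A state machine picks random idles and plays the swing while the owner's melee combo montage runs, then feeds a cached, handle-offset pose.
# A final RigidBody weighted by Is_Swinging follows a blend fixed fully on that pose.
# Other physics chains are unconnected or zero-weighted, leaving the computed rigid alphas without visible effect.
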AnimBlueprint /BRCosmetics/Weapons/FORT_Melee/Pickaxe_Ruckus_Mini/Meshes/Ruckus_Mini_Axe_AnimBP
target_skeleton=&/Ruckus_Mini_Axe_Skeleton
graphs [
 AnimGraph {nodes [
   Local To Component(LocalPose=Input Pose())
   "Transform (Modify) Bone - Bone: b_AR_Root"(
    ComponentPose=Local To Component(LocalPose=Sequence Player 'Pickaxe_Ruckus_Mini_Swing_PROP'() {sequence=/BRCosmetics/Weapons/FORT_Melee/Pickaxe_Ruckus_Mini/Animation/Pickaxe_Ruckus_Mini_Swing_PROP})
    Alpha=1
    Translation="0,0,0"
    Rotation="0,90,0"
    Scale="1,1,1"
   ) {
    rotationMode=BMM_Additive
    rotationSpace=BCS_BoneSpace
    alphaBoolBlend {blendOption=Linear}
    boneToModify {boneName=b_AR_Root}
    rotation {yaw=90}
   }
   "Transform (Modify) Bone - Bone: b_AR_Root"(Alpha=1, Translation="0,0,0", Rotation="0.000,80.000,0.000", Scale="1,1,1") {
    rotationMode=BMM_Additive
    rotationSpace=BCS_BoneSpace
    alphaBoolBlend {blendOption=Linear}
    boneToModify {boneName=b_AR_Root}
    rotation {yaw=79.9999}
   }
   RigidBody() {
    bClampLinearTranslationLimitToRefPose=true
    bOverrideWorldGravity=true
    overridePhysicsAsset=&/Ruckus_Mini_Axe_Physics
    simulationSpace=WorldSpace
    alphaBoolBlend {blendOption=Linear}
    alphaScaleBiasClamp {interpSpeedDecreasing=8, interpSpeedIncreasing=8}
    baseBoneRef {boneName=Melee_TwoHanded_Handle}
    componentAppliedLinearAccClamp=<300 300 300>
    componentLinearAccScale=<0.4 0.3 0.3>
    componentLinearVelScale=<0.4 0.3 0.3>
    overrideWorldGravity {z=-20}
   }
   RigidBody(Alpha=Get Rigid Alpha() -> (Rigid Alpha:double)) {
    bClampLinearTranslationLimitToRefPose=true
    bOverrideWorldGravity=true
    overridePhysicsAsset=&Ruckus_Mini_Axe_Physics
    simulationSpace=WorldSpace
    alphaBoolBlend {blendOption=Linear}
    alphaScaleBiasClamp {bInterpResult=true, interpSpeedDecreasing=1, interpSpeedIncreasing=1, scale=0.15}
    baseBoneRef {boneName=spine_MSTR}
    componentAppliedLinearAccClamp=<3000 3000 3000>
    componentLinearAccScale=<0.6 0.3 0.3>
    componentLinearVelScale=<0.6 0.3 0.3>
    overrideWorldGravity {z=-980}
   }
   n1 = "Transform (Modify) Bone - Bone: b_AR_Root"(
    ComponentPose=Local To Component(LocalPose=Sequence Player 'Pickaxe_Ruckus_Mini_Idle_PROP'() {sequence=/BRCosmetics/Weapons/FORT_Melee/Pickaxe_Ruckus_Mini/Animation/Pickaxe_Ruckus_Mini_Idle_PROP})
    Alpha=1
    Translation="8,8,0"
    Rotation="0,90,0"
    Scale="1,1,1"
   ) {
    rotationMode=BMM_Additive
    rotationSpace=BCS_BoneSpace
    alphaBoolBlend {blendOption=Linear}
    boneToModify {boneName=b_AR_Root}
    rotation {yaw=90}
    translation=<8 8>
   }
   n2 = RigidBody(ComponentPose=@n1, Alpha=Get Rigid Alpha()) {
    bClampLinearTranslationLimitToRefPose=true
    bOverrideWorldGravity=true
    overridePhysicsAsset=&Ruckus_Mini_Axe_Physics
    simulationSpace=WorldSpace
    alphaBoolBlend {blendOption=Linear}
    alphaScaleBiasClamp {interpSpeedDecreasing=8, interpSpeedIncreasing=8}
    baseBoneRef {boneName=Melee_TwoHanded_Handle}
    componentAppliedLinearAccClamp=<300 300 300>
    componentLinearAccScale=<0.4 0.3 0.3>
    componentLinearVelScale=<0.4 0.3 0.3>
    overrideWorldGravity {z=-20}
   }
   n3 = RigidBody(ComponentPose=@n2, Alpha=0.4) {
    bClampLinearTranslationLimitToRefPose=true
    bOverrideWorldGravity=true
    overridePhysicsAsset=&/Ruckus_Mini_Axe_Run_Physics
    simulationSpace=WorldSpace
    alphaBoolBlend {blendOption=Linear}
    alphaScaleBiasClamp {interpSpeedDecreasing=8, interpSpeedIncreasing=5}
    baseBoneRef {boneName=spine_MSTR}
    overrideWorldGravity {z=-980}
   }
   n4 = Two Way Blend(A=Use cached pose 'Anim'(), B=Component To Local(ComponentPose=@n3), Alpha=0) {blendNode {alphaBoolBlend {blendOption=Linear}, alphaScaleBiasClamp {bInterpResult=true, interpSpeedDecreasing=5, interpSpeedIncreasing=5}}}
   n5 = RigidBody(ComponentPose=Local To Component(LocalPose=@n4), Alpha=Get Is_Swinging() -> (Is_Swinging:double)) {
    bClampLinearTranslationLimitToRefPose=true
    bOverrideWorldGravity=true
    overridePhysicsAsset=&Ruckus_Mini_Axe_Physics
    simulationSpace=BaseBoneSpace
    alphaBoolBlend {blendOption=Linear}
    alphaScaleBiasClamp {bClampResult=true, interpSpeedDecreasing=8, interpSpeedIncreasing=8}
    baseBoneRef {boneName=b_AR_Root}
    componentAppliedLinearAccClamp=<3000 3000 3000>
    componentLinearAccScale {z=0.3}
    componentLinearVelScale {z=0.3}
    overrideWorldGravity {z=980}
   }
   Output Pose(Result=Component To Local(ComponentPose=@n5)) {layerGroup=None}
   AnimGraphNode_StateMachine_0 = Idle State Machine() {is=title AnimGraphNode_StateMachine}
   n6 = "Transform (Modify) Bone - Bone: Melee_TwoHanded_Handle"(
    ComponentPose=Local To Component(LocalPose=@AnimGraphNode_StateMachine_0)
    Alpha=1
    Translation="-8,8,0.000"
    Rotation="0,90,0"
    Scale="1,1,1"
   ) {
    rotationMode=BMM_Additive
    rotationSpace=BCS_ParentBoneSpace
    translationMode=BMM_Additive
    translationSpace=BCS_ParentBoneSpace
    alphaBoolBlend {blendOption=Linear}
    alphaScaleBiasClamp {interpSpeedDecreasing=5, interpSpeedIncreasing=6}
    boneToModify {boneName=Melee_TwoHanded_Handle}
    rotation {yaw=90}
    translation=<-8 8>
   }
   Save cached pose 'Anim'(Pose=Component To Local(ComponentPose=@n6)) {cacheName=Anim}
  ]}
 Idle State Machine {
  entry=Idle
  path=AnimGraph.AnimGraphNode_StateMachine_0.Idle State Machine
  ids=id
  nodes [
   AnimStateNode_0 = Idle()
   AnimStateNode_1 = Striking() {is=title AnimStateNode}
   AnimStateTransitionNode_0 = AnimStateTransitionNode()
   AnimStateTransitionNode_1 = AnimStateTransitionNode() {bAutomaticRuleBasedOnSequencePlayerInState=true}
  ]
  links [
   {
    dst=Striking
    rule=AnimGraph.AnimGraphNode_StateMachine_0.Idle State Machine.AnimStateTransitionNode_0.Transition
    src=Idle
    via=AnimStateTransitionNode_0
   }
   {
    dst=Idle
    rule=AnimGraph.AnimGraphNode_StateMachine_0.Idle State Machine.AnimStateTransitionNode_1.Transition
    src=Striking
    via=AnimStateTransitionNode_1
   }
  ]
 }
 Idle {
  path=AnimGraph.AnimGraphNode_StateMachine_0.Idle State Machine.AnimStateNode_0.Idle
  nodes [
   n1 = Random Sequence Player!() -> (Pose:Pose Link Structure) {
    is=title AnimGraphNode_RandomPlayer
    entries [
     {
      chanceToPlay=0.5
      maxLoopCount=0
      maxPlayRate=1
      minLoopCount=0
      minPlayRate=1
      sequence=&Pickaxe_Ruckus_Mini_Idle_PROP
      blendIn {blendOption=Linear, blendTime=0.2, customCurve=null}
     }
     {
      chanceToPlay=0.5
      maxLoopCount=0
      maxPlayRate=1
      minLoopCount=0
      minPlayRate=1
      sequence=/BRCosmetics/Weapons/FORT_Melee/Pickaxe_Ruckus_Mini/Animation/Pickaxe_Ruckus_Mini_idleBreak_PROP
      blendIn {blendOption=Linear, blendTime=0.2, customCurve=null}
     }
    ]
   }
   Output Animation Pose(Result=@n1)
  ]
 }
 Striking {
  path=AnimGraph.AnimGraphNode_StateMachine_0.Idle State Machine.AnimStateNode_1.Striking
  nodes [Output Animation Pose(Result=Sequence Player 'Pickaxe_Ruckus_Mini_Swing_PROP'() {sequence=&Pickaxe_Ruckus_Mini_Swing_PROP})]
 }
 Transition {
  path=AnimGraph.AnimGraphNode_StateMachine_0.Idle State Machine.AnimStateTransitionNode_0.Transition
  nodes [Result(bCanEnterTransition="float > float"(A=Get Is_Swinging(), B=0) {is=title K2Node_CallFunction})]
 }
 Transition {path=AnimGraph.AnimGraphNode_StateMachine_0.Idle State Machine.AnimStateTransitionNode_1.Transition, nodes [Result()]}
 EventGraph {nodes [
   Event BlueprintUpdateAnimation()
   > n1 = Sequence!(execute) -> (then_0, then_1, then_2:Exec)
   n2 = Get bIsPlayingMeleeAnim()
   n3 = Get bIsSprinting()
   > n4 = Branch(Condition=AND Boolean!(A=Get bIsInFrontEnd_0() -> (bIsInFrontEnd_0:bool), B=NOT Boolean(A=@n3), C:bool=NOT Boolean(A=@n2)) -> (ReturnValue))
   > Set Rigid Alpha(Rigid Alpha:double=0.3) -> (Output_Get:double)
   Set Rigid Alpha(execute=@n4.else, Rigid Alpha=0.3)
   n5 = GetOwnerFortPlayerAnimInstance!(self:Custom Character Part Anim Instance Object Reference) -> (ReturnValue:Fort Player Anim Instance Object Reference)
   Set Is_Swinging(
    execute=@n1.then_1
    Is_Swinging:double="To Float (Boolean)"(InBool=Montage_IsPlaying(self=@n5, Montage=/Game/Animation/Game/MainPlayer/Montages/Melee/PlayerMeleeCombo))
   ) -> (Output_Get:double)
   n6 = NOT Boolean(A=Get bIsTacticalSprinting())
   n7 = Get bIsSprinting()
   > Set sprint(sprint:double=SelectFloat(A=1, B=0, bPickA=AND Boolean!(A=@n7, B=@n6) -> (ReturnValue))) -> (Output_Get:double)
   n8 = Branch(execute=@n1.then_2, Condition=Get bIsInFrontEnd_0())
   > Set Rigid Alpha_Run(Rigid Alpha_Run:double=0.1) -> (Output_Get:double)
   Set Rigid Alpha_Run(execute=@n8.else, Rigid Alpha_Run=1)
   TryGetPawnOwner()
  ]}
]
variables [
 bIsInFrontEnd_0 {type=boolean}
 Is_Swinging {type=number, default=1}
 Rigid Alpha {type=number}
 Rigid Alpha_Run {type=number}
 sprint {type=number}
 swing {type=boolean}
]
